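# Translucent material for the Marionette emote's spotlight effect, usable on Niagara mesh particles, with particle color feeding an emissive adjustment function.
# A beam mask, selectable between procedural gradient and ramp texture, blends with two panning smoke textures scaled by the Smoke_light dynamic parameter.
# An optional vertical mask shapes the result, which is multiplied by particle alpha.
# Near-camera and view-angle fades soften opacity, with depth fade skipped on mobile.
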Material /Game/Animation/Game/MainPlayer/Emotes/Marionette/FX/M_Emote_Marionette_SpotLight
graph {
 properties {
  bUsedWithInstancedStaticMeshes=true
  bUsedWithNiagaraMeshParticles=true
  blendMode=BLEND_Translucent
  pixelDepthOffsetMode=PDOM_Legacy
  usedShadingModels=Default Lit
 }
 nodes [
  n1 = ParticleColor()
  MaterialFunctionCall(Emissive=@n1.RGB) {materialFunction=&MF_FX_EmissiveAdjustments}
  n2 = ScalarParameter() {defaultValue=200, parameterName=Near Fade}
  n3 = ScalarParameter() {defaultValue=4, parameterName=Angle Fade}
  n4 = MaterialFunctionCall() {materialFunction=&LinearGradient}
  n5 = OneMinus(None=@n4.VGradient)
  n6 = Multiply(A=Multiply(A=@n5, B=@n4.VGradient)) {constB=1.5}
  n7 = TextureSample() {texture=/Game/Effects/Fort_Effects/Textures/Ramps/T_Ramp_Vert_Soft_Fade}
  n8 = ScalarParameter() {defaultValue=1, parameterName=UseTextureGradient}
  n9 = TextureSample(UVs=TextureCoordinate() {vTiling=-1}) {texture=&T_Ramp_Vert_Soft_Fade}
  n10 = LinearInterpolate(A=@n5, B=@n9.G, Alpha=@n8)
  n11 = VectorParameter() {parameterName=Smoke A UV, sortPriority=1, defaultValue=<rgba 2 2 -0.05 0.2>}
  n12 = Panner(Coordinate=Multiply(A=TextureCoordinate(), B=AppendVector(A=@n11.R, B=@n11.R)), Speed=AppendVector(A=@n11.B, B=@n11.B))
  n13 = VectorParameter() {parameterName=Smoke B UV, sortPriority=1, defaultValue=<rgba 1 1 0.05 0.15>}
  n14 = Multiply(A=TextureCoordinate(), B=AppendVector(A=@n13.R, B=@n13.R))
  n15 = AppendVector(A=@n13.B, B=@n13.B)
  n16 = LinearInterpolate(
   A=Power(Base=TextureSample(UVs=@n12) {texture=/Game/Global/Textures/Default/Tileable_Patterns/T_Smoke_Puff_Tile_Up}.R)
   B=TextureSample(UVs=Panner(Coordinate=@n14, Speed=@n15)) {texture=/Game/Global/Textures/Default/Tileable_Patterns/T_Thick_Smoke_05}.R
  )
  n17 = DynamicParameter() {paramNames [Smoke_light,Param2,Param3,Param4]}
  n18 = LinearInterpolate(
   A=LinearInterpolate(A=@n6, B=@n7.G, Alpha=@n8)
   B=Multiply(A=Power(Base=@n10) {constExponent=0.5}, B=Multiply(A=@n16, B=@n17.Smoke_light))
   Alpha=ScalarParameter() {defaultValue=0.5, parameterName=Lerp Smoke}
  ) {constAlpha=0.8}
  n19 = Multiply(A=Multiply(A=@n4.VGradient, B=Multiply(A=@n4.VGradient, B=@n5))) {constB=1.5}
  n20 = Multiply(A=@n18, B=Multiply(A=LinearInterpolate(A=@n6, B=@n7.G, Alpha=@n8), B=LinearInterpolate(A=@n19, B=@n9.G, Alpha=@n8)))
  n21 = ScalarParameter() {parameterName=UseVerticalMask}
  n22 = Multiply(
   A=Multiply(A=@n1.A, B=LinearInterpolate(A=@n18, B=@n20, Alpha=@n21) {constAlpha=0})
   B=ScalarParameter() {defaultValue=0.25, parameterName=Alpha Mult}
  )
  n23 = DepthFade(FadeDistance=ScalarParameter() {defaultValue=60, parameterName=Depth Fade})
  Multiply(
   A=Multiply(
    A=MaterialFunctionCall(Distance=@n2) {materialFunction=&Near_Camera_Fade}.Result
    B=MaterialFunctionCall("FadeSensitivity (optional)"=@n3) {materialFunction=&FadeBasedOnViewAngle}.Result
   )
   B=ShadingPathSwitch(Default=Multiply(A=@n22, B=@n23), Mobile=@n22) {desc=Migrated from FeatureLevelSwitch}
  )
 ]
}
